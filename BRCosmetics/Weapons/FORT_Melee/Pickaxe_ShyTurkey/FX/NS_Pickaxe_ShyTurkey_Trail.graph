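# Niagara system for the ShyTurkey pickaxe swing trail, made of a thin blue and a wide, faint yellow ribbon emitter.
# Both use shared AnimTrail modules to spawn and update particles from bone/socket indices, with colors tinted by time of day.
# ScaleColor fades them over particle life and by loop-age curves switched on User.SwingSectionIndex.
# Blue also applies System.PerSwingFade and sets ribbon facing from the system X axis; yellow uses its own cutoff curve.
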NiagaraSystem /BRCosmetics/Weapons/FORT_Melee/Pickaxe_ShyTurkey/FX/NS_Pickaxe_ShyTurkey_Trail
dependencies {
 usedDataInterfaces [&NiagaraDataInterfaceCurve, &NiagaraDataInterfaceSocketReader, &NiagaraDataInterfaceColorCurve]
 usedDynamicInputs [&ColorFromCurve]
 usedModules [
  &self:SystemSpawnScript.SystemScriptSource.SystemScriptGraph.NiagaraNodeAssignment_1.SetVariables_#1
  &SystemState
  &self:SystemSpawnScript.SystemScriptSource.SystemScriptGraph.NiagaraNodeAssignment_0.SetVariables_#2
  &self:SystemSpawnScript.SystemScriptSource.SystemScriptGraph.NiagaraNodeAssignment_2.SetVariables_#3
 ]
}
emitters {
 BlueTrail002 {
  emitterUpdateScript [
   &EmitterState(Life Cycle Mode=::NewEnumerator1, Loop Behavior=::NewEnumerator1, Loop Duration=1, UseLoopDelay=false)
   &FN_SpawnAnimTrail(Spawn Interval=0)
  ]
  particleSpawnScript [
   &FN_InitializeAnimTrailBoneIndecies(First Bone or Socket Index=1, Second Bone or Socket Index=5)
   &FN_InitializeAnimTrail(
    AnimTrailWidth=1.7
    Color=&MultiplyLinearColor(
     Linear Color=<rgba 0.0282 0.0858 0.4696 1>
     Scale Factor=&MakeVector4_FromLinearColor(LinearColor=$System.TODColorMultiplier)
    )
    Lifetime=0.16
   )
  ]
  particleUpdateScript [
   &FN_UpdateAnimTrail() {module=FN_UpdateAnimTrail001}
   &ScaleColor(
    Scale Alpha=&Multiply_Float(
     A=&Multiply_Float(
      A=1.25
      B=&Multiply_Float(
       A=&FloatFromCurve(FloatCurve=DataInterface() {propertyValues {Curve {
           defaultValue=0
           keys [
            {arriveTangent=-3.4649, interpMode=RCIM_Cubic, leaveTangent=-3.4649, tangentMode=RCTM_User, time=0, value=1}
            {interpMode=RCIM_Cubic, tangentMode=RCTM_User, time=1, value=0}
           ]
          }}})
       B=&MergeHelper_FloatSwitch(
        0=&FloatFromCurve(
         FloatCurve=DataInterface() {propertyValues {Curve {defaultValue=1, keys [{tangentMode=RCTM_User, time=0.08, value=0}, {tangentMode=RCTM_User, time=0.1, value=1}]}}}
         CurveIndex=$System.NormalizedLoopAge
        )
        1=&FloatFromCurve(
         FloatCurve=DataInterface() {propertyValues {Curve {defaultValue=1, keys [{tangentMode=RCTM_User, time=0, value=0}, {tangentMode=RCTM_User, time=0.1, value=1}]}}}
         CurveIndex=$System.NormalizedLoopAge
        )
        2=&FloatFromCurve(
         FloatCurve=DataInterface() {propertyValues {Curve {defaultValue=1, keys [{tangentMode=RCTM_User, time=0, value=0}, {tangentMode=RCTM_User, time=0.1, value=1}]}}}
         CurveIndex=$System.NormalizedLoopAge
        )
        3=&FloatFromCurve(
         FloatCurve=DataInterface() {propertyValues {Curve {defaultValue=1, keys [{tangentMode=RCTM_User, time=0, value=0}, {tangentMode=RCTM_User, time=0.1, value=1}]}}}
         CurveIndex=$System.NormalizedLoopAge
        )
        Index=$User.SwingSectionIndex
        Num=::NewEnumerator2
       )
      )
     )
     B=$System.PerSwingFade
    )
    ScaleA=true
    ScaleRGB=true
   )
   /SetVariables(Particles.RibbonFacing=/Game/Effects/Niagara/DynamicInputs/Engine/FN_GetSystemAxis(Axis:ENiagaraOrientationAxis=::XAxis, Flip Axis:NiagaraBool=false)) {isSetParameters=true, module=SetVariables_#4}
  ]
  rendererClasses [&NiagaraRibbonRendererProperties]
  renderers [
   {rendererClass=&NiagaraRibbonRendererProperties}
   {rendererClass=&NiagaraRibbonRendererProperties, rendererIndex=1}
  ]
 }
 YellowTrail {
  emitterUpdateScript [
   &EmitterState(Life Cycle Mode=::NewEnumerator1, Loop Behavior=::NewEnumerator1, Loop Duration=1, UseLoopDelay=false)
   &FN_SpawnAnimTrail(Spawn Interval=0)
  ]
  particleSpawnScript [
   &FN_InitializeAnimTrailBoneIndecies(First Bone or Socket Index=4, Second Bone or Socket Index=4)
   &FN_InitializeAnimTrail(
    AnimTrailWidth=20
    Color=&MultiplyLinearColor(
     Linear Color=<rgba 0.8754 0.5982 0 0.2>
     Scale Factor=&MakeVector4_FromLinearColor(LinearColor=&FNTimeOfDayColor(
       Night Color=<rgba 0.2675 0.3179 0.4774 1>
       Morning Color=<rgba 1 0.6182 0.5361 1>
       Day Color=<rgba 1 1 1 1>
       Evening Color=<rgba 1 0.6182 0.5361 1>
      ))
    )
    Lifetime=0.2
   )
  ]
  particleUpdateScript [
   &FN_UpdateAnimTrail() {module=FN_UpdateAnimTrail001}
   &ScaleColor(
    Scale Alpha=&Multiply_Float(
     A=&Multiply_Float(
      A=0.75
      B=&Multiply_Float(
       A=&FloatFromCurve(FloatCurve=DataInterface() {propertyValues {Curve {
           defaultValue=0
           keys [
            {arriveTangent=0.4646, interpMode=RCIM_Cubic, leaveTangent=0.4646, tangentMode=RCTM_User, time=0, value=1}
            {interpMode=RCIM_Cubic, tangentMode=RCTM_User, time=1, value=0}
           ]
          }}})
       B=&MergeHelper_FloatSwitch(
        0=&FloatFromCurve(
         FloatCurve=DataInterface() {propertyValues {Curve {defaultValue=1, keys [{tangentMode=RCTM_User, time=0.08, value=0}, {tangentMode=RCTM_User, time=0.1, value=1}]}}}
         CurveIndex=$System.NormalizedLoopAge
        )
        1=&FloatFromCurve(
         FloatCurve=DataInterface() {propertyValues {Curve {defaultValue=1, keys [{tangentMode=RCTM_User, time=0, value=0}, {tangentMode=RCTM_User, time=0.1, value=1}]}}}
         CurveIndex=$System.NormalizedLoopAge
        )
        2=&FloatFromCurve(
         FloatCurve=DataInterface() {propertyValues {Curve {defaultValue=1, keys [{tangentMode=RCTM_User, time=0, value=0}, {tangentMode=RCTM_User, time=0.1, value=1}]}}}
         CurveIndex=$System.NormalizedLoopAge
        )
        3=&FloatFromCurve(
         FloatCurve=DataInterface() {propertyValues {Curve {defaultValue=1, keys [{tangentMode=RCTM_User, time=0, value=0}, {tangentMode=RCTM_User, time=0.1, value=1}]}}}
         CurveIndex=$System.NormalizedLoopAge
        )
        Index=$User.SwingSectionIndex
        Num=::NewEnumerator2
       )
      )
     )
     B=&FloatFromCurve(FloatCurve=DataInterface() {propertyValues {Curve {keys [<at 0.12 1>,<at 0.15 0>]}}})
    )
    ScaleA=true
    ScaleRGB=true
   )
   /SetVariables() {isSetParameters=true, module=SetVariables_#4}
  ]
  rendererClasses [&NiagaraRibbonRendererProperties]
  renderers [
   {rendererClass=&NiagaraRibbonRendererProperties}
   {rendererClass=&NiagaraRibbonRendererProperties, rendererIndex=1}
  ]
 }
}
summary {
 systemName=NS_Pickaxe_ShyTurkey_Trail
 emitters [
  {emitterName=YellowTrail, rendererClasses [&NiagaraRibbonRendererProperties]}
  {emitterName=BlueTrail002, rendererClasses [&NiagaraRibbonRendererProperties]}
 ]
 userVariables [
  User.bIsFrontEnd {type=NiagaraBool}
  User.bIsFrontEndPreview {type=NiagaraBool}
  User.PhysicalSurface {type=NiagaraInt32}
  User.RibbonIDOffset {type=NiagaraInt32}
  User.SwingSectionIndex {type=NiagaraInt32}
 ]
}
